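; HitTicksDynamicUIDirector is a Dynamic UI Director that registers the HitTicks UI scene for the local player.
; On BeginPlay it gets the owning local player controller's Fort global UI context and checks whether it is in a zone.
; Only when that check passes does it add HitTicksDynamicUIScene, with no other logic or construction-script work.
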(event EventBeginPlay
  (if (GlobalUIContext|IsinZone (PlayerController|LocalPlayerSubsystems|GetFortGlobalUIContext (DynamicUIDirector|GetOwningLocalPlayerController)))
    (DynamicUIDirector|AddScene "/HitTicks/HitTicksDynamicUIScene.HitTicksDynamicUIScene")))

(fn ConstructionScript)
