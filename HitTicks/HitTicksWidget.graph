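# Hit tick HUD widget, built on native FortUserWidget_HitTicks, showing a material-driven hit marker, a shield icon and shield-break imagery.
# On Construct it zeroes tick opacity, then binds and immediately calls HUD scale, reticle settings and pawn handlers.
# Animations drive material parameters, opacity and scale for shield-crack and elimination effects; the graphs never play them.
# InitializeTicks also holds an unconnected chain that offsets four corner ticks absent from the widget tree.
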WidgetBlueprint /HitTicks/HitTicksWidget
animations [
 Anim_ShieldCrack {
  binding_count=5
  length_seconds=1
  object=&self:Anim_ShieldCrack
  sequence {bindings [
    ShieldIcon {
     id=#1
     possessed_class=Image
     tracks [
      {
       type=MovieSceneWidgetMaterialTrack
       sections [{
         type=MovieSceneParameterSection
         end_frame=20
         start_frame=0
         channels [EnableCracks {type=MovieSceneScriptingFloatChannel, keys [[4.762,1],[9.5237,0]]}]
        }]
      }
      {
       type=MovieSceneFloatTrack
       display_name=Render Opacity
       sections [{
         type=MovieSceneFloatSection
         end_frame=11
         start_frame=0
         channels [None {type=MovieSceneScriptingFloatChannel, keys [[0,0],[4.762,1],[9.5237,1],[11.4287,0]]}]
        }]
      }
     ]
    }
    Shieldbreak_base {
     id=#2
     possessed_class=Image
     tracks [
      {
       type=MovieScene2DTransformTrack
       sections [{
         type=MovieScene2DTransformSection
         end_frame=16
         start_frame=0
         channels [
          Translation.X {type=MovieSceneScriptingFloatChannel, const [0,1]}
          Translation.Y {type=MovieSceneScriptingFloatChannel, const [0,1]}
          Angle {type=MovieSceneScriptingFloatChannel, const [0,1]}
          Scale.X {type=MovieSceneScriptingFloatChannel, keys [[0,1.1],[3,1.3],[10,1.3]]}
          Scale.Y {type=MovieSceneScriptingFloatChannel, keys [[0,1.1],[3,1.3],[10,1.3]]}
          Shear.X {type=MovieSceneScriptingFloatChannel, const [0,1]}
          Shear.Y {type=MovieSceneScriptingFloatChannel, const [0,1]}
         ]
        }]
      }
      {
       type=MovieSceneColorTrack
       display_name=Color and Opacity
       sections [{
         type=MovieSceneColorSection
         end_frame=20
         start_frame=0
         channels [
          Color.R {type=MovieSceneScriptingFloatChannel, const [1,1]}
          Color.G {type=MovieSceneScriptingFloatChannel, const [1,1]}
          Color.B {type=MovieSceneScriptingFloatChannel, const [1,1]}
          Color.A {type=MovieSceneScriptingFloatChannel, keys [[0,0],[1,0],[3,1],[12,0.75],[18,0]]}
         ]
        }]
      }
     ]
    }
    HitMarker {
     id=#3
     possessed_class=Image
     tracks [{
       type=MovieSceneFloatTrack
       display_name=Render Opacity
       sections [{type=MovieSceneFloatSection, end_frame=18, start_frame=0, channels [None {type=MovieSceneScriptingFloatChannel, keys [[10,0],[15,1]]}]}]
      }]
    }
    OverlaySlot_1 {id=#4, parent_id=#2, possessed_class=OverlaySlot, tracks []}
    Shieldbreak_glow {
     id=#5
     possessed_class=Image
     tracks [
      {
       type=MovieScene2DTransformTrack
       sections [{
         type=MovieScene2DTransformSection
         end_frame=15
         start_frame=0
         channels [
          Translation.X {type=MovieSceneScriptingFloatChannel, const [0,1]}
          Translation.Y {type=MovieSceneScriptingFloatChannel, const [0,1]}
          Angle {type=MovieSceneScriptingFloatChannel, const [0,1]}
          Scale.X {type=MovieSceneScriptingFloatChannel, keys [[0,1.1],[3,1.3],[10,1.3]]}
          Scale.Y {type=MovieSceneScriptingFloatChannel, keys [[0,1.1],[3,1.3],[10,1.3]]}
          Shear.X {type=MovieSceneScriptingFloatChannel, const [0,1]}
          Shear.Y {type=MovieSceneScriptingFloatChannel, const [0,1]}
         ]
        }]
      }
      {
       type=MovieSceneColorTrack
       display_name=Color and Opacity
       sections [{
         type=MovieSceneColorSection
         end_frame=20
         start_frame=0
         channels [
          Color.R {type=MovieSceneScriptingFloatChannel, const [1,1]}
          Color.G {type=MovieSceneScriptingFloatChannel, const [1,1]}
          Color.B {type=MovieSceneScriptingFloatChannel, const [1,1]}
          Color.A {type=MovieSceneScriptingFloatChannel, keys [[0,0],[1,0],[3,0.7],[5,1],[6,1],[18,0]]}
         ]
        }]
      }
     ]
    }
   ]}
 }
 Anim_Elimination {
  binding_count=1
  length_seconds=0.35
  object=&self:Anim_Elimination
  sequence {bindings [HitMarker {
     id=#3
     possessed_class=Image
     tracks [{
       type=MovieSceneWidgetMaterialTrack
       sections [{
         type=MovieSceneParameterSection
         end_frame=7
         start_frame=0
         channels [
          MoveTicks {type=MovieSceneScriptingFloatChannel, const [0.3,1]}
          Opacity {type=MovieSceneScriptingFloatChannel, keys [[0,1],[4,1],[7,0]]}
         ]
        }]
      }]
    }]}
 }
 Anim_Elimination2 {
  binding_count=3
  length_seconds=1.1
  object=&self:Anim_Elimination2
  sequence {bindings [
    PanelSlot_0 {id=#6, parent_id=#7, possessed_class=PanelSlot, tracks []}
    Overlay_0 {
     id=#7
     possessed_class=Overlay
     tracks [
      {
       type=MovieScene2DTransformTrack
       sections [{
         type=MovieScene2DTransformSection
         end_frame=10
         start_frame=-1
         channels [
          Translation.X {type=MovieSceneScriptingFloatChannel, const [0,1]}
          Translation.Y {type=MovieSceneScriptingFloatChannel, const [0,1]}
          Angle {type=MovieSceneScriptingFloatChannel, const [0,1]}
          Scale.X {type=MovieSceneScriptingFloatChannel, keys [[-0.0003,1],[10,1.3]]}
          Scale.Y {type=MovieSceneScriptingFloatChannel, keys [[-0.0003,1],[10,1.3]]}
          Shear.X {type=MovieSceneScriptingFloatChannel, const [0,1]}
          Shear.Y {type=MovieSceneScriptingFloatChannel, const [0,1]}
         ]
        }]
      }
      {
       type=MovieSceneFloatTrack
       display_name=Render Opacity
       sections [{
         type=MovieSceneFloatSection
         end_frame=22
         start_frame=0
         channels [None {type=MovieSceneScriptingFloatChannel, keys [[0,0],[2,1],[6,1],[10,0],[20,0]]}]
        }]
      }
     ]
    }
    HitMarker {id=#8, possessed_class=Image, tracks []}
   ]}
 }
]
graphs [
 EventGraph {nodes [
   Event Construct()
   > InitializeTicks()
   > Bind Event to On HUDScale Changed(
    self:Fort HUDContext Object Reference=Get FortHUDContext!() -> (ReturnValue:Fort HUDContext Object Reference) {is=title K2Node_GetBlueprintContextEx}
    Delegate=Create Event()
   ) {is=title K2Node_AddDelegate}
   > HudScaleChanged(self:Fort User Widget Hit Ticks Object Reference)
   > Bind Event to On Reticle Settings Changed(self:Fort HUDContext Object Reference=Get FortHUDContext(), Delegate=Create Event()) {is=title K2Node_AddDelegate}
   > OnReticleSettingsChanged(self:Fort User Widget Hit Ticks Object Reference)
   > Bind Event to On Pawn Set(self:Fort HUDContext Object Reference=Get FortHUDContext(), Delegate=Create Event()) {is=title K2Node_AddDelegate}
   > OnPawnSet(self:Fort User Widget Hit Ticks Object Reference)
  ]}
 InitializeTicks {nodes [
   Set Slot Distance!(
    execute
    self
    Direction:Vector="1,-1, 0"
    Distance:double=Get BaseDistance() -> (BaseDistance:double)
    Tick Image:Image Object Reference=Get HitMarker_TopRight() -> (HitMarker_TopRight:Image Object Reference)
   ) -> (then) {is=title K2Node_CallFunction}
   > Set Slot Distance(Direction="-1,-1, 0", Distance=Get BaseDistance(), Tick Image=Get HitMarker_TopLeft() -> (HitMarker_TopLeft:Image Object Reference))
   > Set Slot Distance(Direction="-1,1, 0", Distance=Get BaseDistance(), Tick Image=Get HitMarker_BottomLeft() -> (HitMarker_BottomLeft:Image Object Reference))
   > Set Slot Distance(Direction="1,1, 0", Distance=Get BaseDistance(), Tick Image=Get HitMarker_BottomRight() -> (HitMarker_BottomRight:Image Object Reference))
   InitializeTicks() {is=title K2Node_FunctionEntry}
   > UpdateTickMarkerOpacity!(execute, self:Fort User Widget Hit Ticks Object Reference, NewOpacity:float=0) -> (then)
  ]}
 NewEventDispatcher {nodes [NewEventDispatcher() {is=title K2Node_FunctionEntry}]}
 NewEventDispatcher_0 {nodes [NewEventDispatcher_0() {is=title K2Node_FunctionEntry}]}
 Set Slot Distance {nodes [
   n1 = Set Slot Distance!() -> (then, Direction:Vector=~, Distance:double=~, Tick Image:Image Object Reference) {is=title K2Node_FunctionEntry}
   > SetRenderTranslation(
    self=@n1.Tick Image
    Translation="To Vector2D (Vector)"!(InVector:Vector=vector * vector(A=@n1, B=@n1.Distance) -> (ReturnValue=~)) -> (ReturnValue:Vector 2D Structure)
   )
  ]}
]
variables [
 BaseDistance {type=number, default=18}
 CurrentMaterialInstance {type=MaterialInstance, default=null}
 EliminationMaterial {type=MaterialInstance, default=null}
 ElimMaterrialProperties {
  type="map<number>"
  default {
   Corner 01 Pos X=-2
   Corner 01 Pos Y=-2
   Corner 02 Pos X=0.25
   Corner 02 Pos Y=-0.25
   Corner 03 Pos X=0.9
   Corner 03 Pos Y=0.9
   Corner 04 Pos X=-0.25
   Corner 04 Pos Y=0.25
  }
 }
 HealthMaterial {type=MaterialInstance, default=null}
 HealthMaterrialProperties {
  type="map<number>"
  default {
   Corner 01 Pos X=-1
   Corner 01 Pos Y=-1
   Corner 02 Pos X=-1
   Corner 02 Pos Y=-1
   Corner 03 Pos X=0.35
   Corner 03 Pos Y=0.4
   Corner 04 Pos X=0.4
   Corner 04 Pos Y=0.35
  }
 }
 HitsForMaxTickScale {type=number, default=5}
 MaxScale {type=number, default=3}
 NewEventDispatcher {}
 NewEventDispatcher_0 {}
 NewVar {type=MaterialInstanceDynamic}
 ResetTimeThreshold {type=number, default=1.5}
 Shield {type=boolean}
 ShieldMaterial {type=MaterialInstance, default=null}
 ShieldMaterrialProperties {
  type="map<number>"
  default {
   Corner 01 Pos X=0.2
   Corner 01 Pos Y=0.2
   Corner 02 Pos X=0.7
   Corner 02 Pos Y=-0.7
   Corner 03 Pos X=-0.2
   Corner 03 Pos Y=-0.2
   Corner 04 Pos X=-0.7
   Corner 04 Pos Y=0.7
  }
 }
 TickScaleAccumulator {type=number, default=2}
 UseNative {type=boolean}
]
widgets {
 info {
  inheritedWidgetCount=2
  namedSlotCount=1
  parentClass=/Script/HitTicksUI.FortUserWidget_HitTicks
  rootWidgetClass=&Overlay
  widgetCount=8
 }
 widgets [
  Overlay_1 {widgetClassPath=&Overlay, widget=Overlay_1}
  FortHUDElementWrapper {
   widgetClassPath=/Game/Athena/HUD/FortHUDElementWrapper
   bIsVariable=true
   parent=Overlay_1
   widget=FortHUDElementWrapper
   properties {
    elementTag {tagName=HUD.Athena.HitTicks}
    hUDElementTag {gameplayTags [{tagName=Athena.HideForCapture}, {tagName=HUD.Athena.HideForPostGame}]}
   }
   slot {class=OverlaySlot}
  }
  Overlay_0 {widgetClassPath=&Overlay, namedSlotHost=FortHUDElementWrapper, widget=Overlay_0}
  HitMarker {
   widgetClassPath=&Image
   bInherited=true
   bIsVariable=true
   parent=Overlay_0
   widget=HitMarker
   properties {
    visibility=SelfHitTestInvisible
    brush {
     bIsDynamicallyLoaded=false
     drawAs=Image
     imageType=NoImage
     mirroring=NoMirror
     resourceName=None
     resourceObject=/HitTicks/Materials/MI_UI_HitMarker
     tiling=NoTile
     imageSize=<100 101>
     margin=<ltrb 0 0 0 0>
     outlineSettings {
      bUseBrushTransparency=false
      roundingType=HalfHeightRadius
      width=0
      color {colorUseRule=UseColor_Specified, specifiedColor=<rgba 0 0 0 0>}
      cornerRadii=<0 0 0 0>
     }
     tintColor {colorUseRule=UseColor_Specified, specifiedColor=<rgba 1 1 1 1>}
     uVRegion {bIsValid=false, max=<0 0>, min=<0 0>}
    }
    colorAndOpacity=<rgba 1 1 1 0.8>
   }
   slot {class=OverlaySlot}
  }
  ShieldIcon {
   widgetClassPath=&Image
   bInherited=true
   bIsVariable=true
   parent=Overlay_0
   widget=ShieldIcon
   properties {
    visibility=Collapsed
    brush {
     bIsDynamicallyLoaded=false
     drawAs=Image
     imageType=NoImage
     mirroring=NoMirror
     resourceName=None
     resourceObject=/HitTicks/Materials/MI_UI_ShieldWithCrack
     tiling=NoTile
     imageSize=<20 20>
     margin=<ltrb 0 0 0 0>
     outlineSettings {
      bUseBrushTransparency=false
      roundingType=HalfHeightRadius
      width=0
      color {colorUseRule=UseColor_Specified, specifiedColor=<rgba 0 0 0 0>}
      cornerRadii=<0 0 0 0>
     }
     tintColor {colorUseRule=UseColor_Specified, specifiedColor=<rgba 1 1 1 1>}
     uVRegion {bIsValid=false, max=<0 0>, min=<0 0>}
    }
    renderTransform {angle=0, scale=<1.4 1.4>, shear=<0 0>, translation=<75 0>}
   }
   slot {class=OverlaySlot}
  }
  Overlay_shieldbreak {
   widgetClassPath=&Overlay
   parent=Overlay_1
   widget=Overlay_shieldbreak
   slot {class=OverlaySlot, properties {horizontalAlignment=HAlign_Center, verticalAlignment=VAlign_Center}}
  }
  Shieldbreak_base {
   widgetClassPath=&Image
   bIsVariable=true
   parent=Overlay_shieldbreak
   widget=Shieldbreak_base
   properties {
    visibility=SelfHitTestInvisible
    brush {
     bIsDynamicallyLoaded=false
     drawAs=Image
     imageType=NoImage
     mirroring=NoMirror
     resourceName=None
     resourceObject=/HitTicks/Materials/T_UI_hitticks_Shieldbreak_base
     tiling=NoTile
     imageSize=<64 64>
     margin=<ltrb 0 0 0 0>
     outlineSettings {
      bUseBrushTransparency=false
      roundingType=HalfHeightRadius
      width=0
      color {colorUseRule=UseColor_Specified, specifiedColor=<rgba 0 0 0 0>}
      cornerRadii=<0 0 0 0>
     }
     tintColor {colorUseRule=UseColor_Specified, specifiedColor=<rgba 1 1 1 1>}
     uVRegion {bIsValid=false, max=<0 0>, min=<0 0>}
    }
    colorAndOpacity=<rgba 1 1 1 0>
   }
   slot {class=OverlaySlot, properties {horizontalAlignment=HAlign_Center, verticalAlignment=VAlign_Center}}
  }
  Shieldbreak_glow {
   widgetClassPath=&Image
   bIsVariable=true
   parent=Overlay_shieldbreak
   widget=Shieldbreak_glow
   properties {
    visibility=SelfHitTestInvisible
    brush {
     bIsDynamicallyLoaded=false
     drawAs=Image
     imageType=NoImage
     mirroring=NoMirror
     resourceName=None
     resourceObject=/HitTicks/Materials/T_UI_hitticks_Shieldbreak_blur
     tiling=NoTile
     imageSize=<64 64>
     margin=<ltrb 0 0 0 0>
     outlineSettings {
      bUseBrushTransparency=false
      roundingType=HalfHeightRadius
      width=0
      color {colorUseRule=UseColor_Specified, specifiedColor=<rgba 0 0 0 0>}
      cornerRadii=<0 0 0 0>
     }
     tintColor {colorUseRule=UseColor_Specified, specifiedColor=<rgba 1 1 1 1>}
     uVRegion {bIsValid=false, max=<0 0>, min=<0 0>}
    }
    colorAndOpacity=<rgba 1 1 1 0>
   }
   slot {class=OverlaySlot, properties {horizontalAlignment=HAlign_Center, verticalAlignment=VAlign_Center}}
  }
 ]
}
