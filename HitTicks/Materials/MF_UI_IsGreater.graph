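# MF_UI_IsGreater is a HitTicks material function that returns a binary mask: 1 when scalar input A is greater than B, otherwise 0.
# It subtracts B from A, saturates the difference and rounds up with Ceil, giving a strict comparison without an If node.
# Both inputs use their preview values as defaults, so unconnected A and B compare 1 against 0.
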MaterialFunction /HitTicks/Materials/MF_UI_IsGreater
graph {nodes [
  n1 = Subtract(
   A=FunctionInput(Preview=Constant() {r=1}) {bUsePreviewValueAsDefault=true, inputName=A, inputType=FunctionInput_Scalar}
   B=FunctionInput(Preview=Constant()) {bUsePreviewValueAsDefault=true, inputName=B, inputType=FunctionInput_Scalar, sortPriority=1}
  )
  FunctionOutput(None=Ceil(None=Saturate(None=@n1)))
 ]}
