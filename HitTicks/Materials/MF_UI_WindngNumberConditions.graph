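# Material function for HitTicks UI materials that turns a scalar ResultNumber, apparently a winding number going by the asset name, into a sign of 1 or -1.
# It adds the floored third of the input to a saturated, floored one-minus term, then reduces that to 0 or 1 and remaps it.
# This branchless math yields 1 for nonzero inputs from -3 up to but excluding 3, and -1 for zero or anything outside.
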MaterialFunction /HitTicks/Materials/MF_UI_WindngNumberConditions
graph {nodes [
  n1 = FunctionInput(Preview=Constant()) {bUsePreviewValueAsDefault=true, inputName=ResultNumber, inputType=FunctionInput_Scalar}
  n2 = OneMinus(None=@n1)
  n3 = Add(A=Floor(None=Divide(A=@n1) {constB=3}), B=Saturate(None=Floor(None=@n2)))
  n4 = Ceil(None=Divide(A=Abs(None=@n3)))
  FunctionOutput(None=MaterialFunctionCall(Input=Saturate(None=@n4), Input Low=Constant(), Input High=Constant() {r=1}, Target Low=Constant() {r=1}, Target High=Constant() {r=-1}) {materialFunction=&RemapValueRange}.Result)
 ]}
