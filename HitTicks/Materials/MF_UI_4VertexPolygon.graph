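# Material function that draws an antialias-style four-vertex polygon SDF for UI, with vertices and UVs as inputs.
# UVs remap to -1..1; distance is the square root of the minimum squared distance to the four edges via clamped projection.
# Per-edge winding tests using MF_UI_IsGreater and MF_UI_WindngNumberConditions multiply into the inside/outside sign.
# It outputs SDF, thickness-based outline SDF, and smoothstepped Fill and Outline with Glow Min/Max inputs for falloff.
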MaterialFunction /HitTicks/Materials/MF_UI_4VertexPolygon
graph {nodes [
  FunctionOutput(None=NamedRerouteUsage()) {outputName=SDF, sortPriority=3}
  FunctionOutput(None=NamedRerouteUsage()) {outputName=SDF Outline, sortPriority=4}
  n1 = NamedRerouteUsage()
  n2 = NamedRerouteUsage()
  n3 = NamedRerouteUsage()
  FunctionOutput(None=OneMinus(None=SmoothStep(Min=@n1, Max=@n2, Value=@n3))) {outputName=Fill}
  n4 = NamedRerouteUsage()
  n5 = NamedRerouteUsage()
  n6 = NamedRerouteUsage()
  FunctionOutput(None=OneMinus(None=SmoothStep(Min=@n4, Max=@n5, Value=@n6))) {outputName=Outline, sortPriority=2}
  n7 = MaterialFunctionCall(Float2=NamedRerouteUsage()) {materialFunction=&BreakOutFloat2Components}
  n8 = ComponentMask(None=NamedRerouteUsage()) {g=true}
  n9 = SmoothStep(
   Min=MaterialFunctionCall(Float2=NamedRerouteUsage()) {materialFunction=&BreakOutFloat2Components}.G
   Value=ComponentMask(None=NamedRerouteUsage()) {g=true}
  )
  n10 = Add(A=Ceil(None=SmoothStep(Min=@n7.G, Value=@n8)), B=OneMinus(None=Ceil(None=@n9)))
  n11 = NamedRerouteUsage()
  n12 = NamedRerouteUsage()
  n13 = NamedRerouteUsage()
  n14 = NamedRerouteUsage()
  n15 = MaterialFunctionCall!(
   A=Multiply(
    A=MaterialFunctionCall(Float2=@n11) {materialFunction=&BreakOutFloat2Components}.R
    B=MaterialFunctionCall(Float2=@n12) {materialFunction=&BreakOutFloat2Components}.G
   )
   B=Multiply(
    A=MaterialFunctionCall(Float2=@n13) {materialFunction=&BreakOutFloat2Components}.G
    B=MaterialFunctionCall(Float2=@n14) {materialFunction=&BreakOutFloat2Components}.R
   )
  ) {materialFunction=&/MF_UI_IsGreater}
  NamedRerouteDeclaration(None=MaterialFunctionCall!(ResultNumber=Add(A=@n10, B=@n15.Result)) {materialFunction=&/MF_UI_WindngNumberConditions}.Result) {name=s3, nodeColor=<rgba 0 0 0 1>}
  n16 = MaterialFunctionCall(Float2=NamedRerouteUsage()) {materialFunction=&BreakOutFloat2Components}
  n17 = ComponentMask(None=NamedRerouteUsage()) {g=true}
  n18 = SmoothStep(
   Min=MaterialFunctionCall(Float2=NamedRerouteUsage()) {materialFunction=&BreakOutFloat2Components}.G
   Value=ComponentMask(None=NamedRerouteUsage()) {g=true}
  )
  n19 = Add(A=Ceil(None=SmoothStep(Min=@n16.G, Value=@n17)), B=OneMinus(None=Ceil(None=@n18)))
  n20 = NamedRerouteUsage()
  n21 = NamedRerouteUsage()
  n22 = NamedRerouteUsage()
  n23 = NamedRerouteUsage()
  n24 = MaterialFunctionCall(
   A=Multiply(
    A=MaterialFunctionCall(Float2=@n20) {materialFunction=&BreakOutFloat2Components}.R
    B=MaterialFunctionCall(Float2=@n21) {materialFunction=&BreakOutFloat2Components}.G
   )
   B=Multiply(
    A=MaterialFunctionCall(Float2=@n22) {materialFunction=&BreakOutFloat2Components}.G
    B=MaterialFunctionCall(Float2=@n23) {materialFunction=&BreakOutFloat2Components}.R
   )
  ) {materialFunction=&MF_UI_IsGreater}
  NamedRerouteDeclaration(None=MaterialFunctionCall(ResultNumber=Add(A=@n19, B=@n24.Result)) {materialFunction=&MF_UI_WindngNumberConditions}.Result) {name=s4, nodeColor=<rgba 0 0 0 1>}
  n25 = FunctionInput(Preview=TextureCoordinate()) {
   bUsePreviewValueAsDefault=true
   description=The Position Coordinates to measure the box. Default is World Position and can be left default in most cases.
   inputName=UVs
   inputType=FunctionInput_Vector2
  }
  NamedRerouteDeclaration(None=ConstantBiasScale(None=@n25) {bias=-0.5, scale=2}) {name="UV's(-1,1)", nodeColor=<rgba 0 0 0 1>}
  NamedRerouteDeclaration(None=FunctionInput(Preview=Constant2Vector() {g=-1, r=-1}) {bUsePreviewValueAsDefault=true, inputName=vertex_0, inputType=FunctionInput_Vector2, sortPriority=1}) {name=Vertex0, nodeColor=<rgba 0 0 0 1>}
  NamedRerouteDeclaration(None=FunctionInput(Preview=Constant2Vector() {g=-1, r=1}) {bUsePreviewValueAsDefault=true, inputName=vertex_1, inputType=FunctionInput_Vector2, sortPriority=2}) {name=Vertex1, nodeColor=<rgba 0 0 0 1>}
  n26 = Subtract(A=NamedRerouteUsage(), B=NamedRerouteUsage())
  NamedRerouteDeclaration(None=DotProduct(A=@n26, B=@n26)) {name=Distance, nodeColor=<rgba 0 0 0 1>}
  NamedRerouteDeclaration(None=FunctionInput(Preview=Constant2Vector() {g=1, r=1}) {bUsePreviewValueAsDefault=true, inputName=vertex_2, inputType=FunctionInput_Vector2, sortPriority=3}) {name=Vertex2, nodeColor=<rgba 0 0 0 1>}
  NamedRerouteDeclaration(None=Subtract(A=NamedRerouteUsage(), B=NamedRerouteUsage())) {name=w1, nodeColor=<rgba 0 0 0 1>}
  NamedRerouteDeclaration(None=Subtract(A=NamedRerouteUsage(), B=NamedRerouteUsage())) {name=w2, nodeColor=<rgba 0 0 0 1>}
  NamedRerouteDeclaration(None=Subtract(A=NamedRerouteUsage(), B=NamedRerouteUsage())) {name=w3, nodeColor=<rgba 0 0 0 1>}
  n27 = NamedRerouteUsage()
  n28 = Min(A=NamedRerouteUsage(), B=DotProduct(A=@n27, B=@n27))
  n29 = NamedRerouteUsage()
  n30 = DotProduct(A=@n29, B=@n29)
  n31 = NamedRerouteUsage()
  n32 = NamedRerouteUsage()
  n33 = Min(A=Min(A=Min(A=@n28, B=@n30), B=DotProduct(A=@n31, B=@n31)), B=DotProduct(A=@n32, B=@n32))
  NamedRerouteDeclaration(None=Multiply(A=SquareRoot(None=@n33), B=NamedRerouteUsage())) {name=SDF Shape, nodeColor=<rgba 0 0 0 1>}
  NamedRerouteDeclaration(None=Subtract(A=NamedRerouteUsage(), B=NamedRerouteUsage())) {name=w4, nodeColor=<rgba 0 0 0 1>}
  NamedRerouteDeclaration(None=Subtract(A=NamedRerouteUsage(), B=NamedRerouteUsage())) {name=e1, nodeColor=<rgba 0 0 0 1>}
  NamedRerouteDeclaration(None=FunctionInput(Preview=Constant2Vector() {g=1, r=-1}) {bUsePreviewValueAsDefault=true, inputName=vertex_3, inputType=FunctionInput_Vector2, sortPriority=4}) {name=Vertex3, nodeColor=<rgba 0 0 0 1>}
  NamedRerouteDeclaration(None=Subtract(A=NamedRerouteUsage(), B=NamedRerouteUsage())) {name=e2, nodeColor=<rgba 0 0 0 1>}
  NamedRerouteDeclaration(None=Subtract(A=NamedRerouteUsage(), B=NamedRerouteUsage())) {name=e3, nodeColor=<rgba 0 0 0 1>}
  NamedRerouteDeclaration(None=Subtract(A=NamedRerouteUsage(), B=NamedRerouteUsage())) {name=e4, nodeColor=<rgba 0 0 0 1>}
  n34 = NamedRerouteUsage()
  n35 = NamedRerouteUsage()
  n36 = NamedRerouteUsage()
  n37 = NamedRerouteUsage()
  n38 = Saturate(None=Divide(A=DotProduct(A=@n35, B=@n36), B=DotProduct(A=@n37, B=@n37)))
  NamedRerouteDeclaration(None=Subtract(A=NamedRerouteUsage(), B=Multiply(A=@n34, B=@n38))) {name=b1, nodeColor=<rgba 0 0 0 1>}
  n39 = NamedRerouteUsage()
  n40 = NamedRerouteUsage()
  n41 = NamedRerouteUsage()
  n42 = NamedRerouteUsage()
  n43 = Saturate(None=Divide(A=DotProduct(A=@n40, B=@n41), B=DotProduct(A=@n42, B=@n42)))
  NamedRerouteDeclaration(None=Subtract(A=NamedRerouteUsage(), B=Multiply(A=@n39, B=@n43))) {name=b2, nodeColor=<rgba 0 0 0 1>}
  n44 = NamedRerouteUsage()
  n45 = NamedRerouteUsage()
  n46 = NamedRerouteUsage()
  n47 = NamedRerouteUsage()
  n48 = Saturate(None=Divide(A=DotProduct(A=@n45, B=@n46), B=DotProduct(A=@n47, B=@n47)))
  NamedRerouteDeclaration(None=Subtract(A=NamedRerouteUsage(), B=Multiply(A=@n44, B=@n48))) {name=b3, nodeColor=<rgba 0 0 0 1>}
  n49 = NamedRerouteUsage()
  n50 = NamedRerouteUsage()
  n51 = NamedRerouteUsage()
  n52 = NamedRerouteUsage()
  n53 = Saturate(None=Divide(A=DotProduct(A=@n50, B=@n51), B=DotProduct(A=@n52, B=@n52)))
  NamedRerouteDeclaration(None=Subtract(A=NamedRerouteUsage(), B=Multiply(A=@n49, B=@n53))) {name=b4, nodeColor=<rgba 0 0 0 1>}
  n54 = NamedRerouteUsage()
  n55 = FunctionInput(Preview=Constant() {r=0.05}) {
   bUsePreviewValueAsDefault=true
   description=Determines the thickness of the stroke when using one of the outline outputs.
   inputName=Outline Thickness
   inputType=FunctionInput_Scalar
   sortPriority=5
  }
  NamedRerouteDeclaration(None=Subtract(A=Abs(None=@n54), B=@n55) {constB=0.1}) {name=SDF Outline, nodeColor=<rgba 0 0 0 1>}
  n56 = Multiply(A=NamedRerouteUsage(), B=NamedRerouteUsage()) {constA=1, constB=-1}
  n57 = NamedRerouteUsage()
  NamedRerouteDeclaration(None=Multiply(A=Multiply(A=@n56, B=@n57), B=NamedRerouteUsage())) {name=Inner Shape, nodeColor=<rgba 0 0 0 1>}
  n58 = FunctionInput(Preview=Constant()) {
   bUsePreviewValueAsDefault=true
   description=Determines glow falloff strength. Can not go below GlowMin value or results will invert
   inputName=Glow Min
   inputType=FunctionInput_Scalar
   sortPriority=6
  }
  NamedRerouteDeclaration(None=@n58) {name=Glow Min, nodeColor=<rgba 0 0 0 1>}
  n59 = FunctionInput(Preview=Constant()) {
   bUsePreviewValueAsDefault=true
   description=Determines glow falloff strength. Can not go below GlowMin value or results will invert
   inputName=Glow Max
   inputType=FunctionInput_Scalar
   sortPriority=7
  }
  NamedRerouteDeclaration(None=@n59) {name=Glow Max, nodeColor=<rgba 0 0 0 1>}
  n60 = MaterialFunctionCall(Float2=NamedRerouteUsage()) {materialFunction=&BreakOutFloat2Components}
  n61 = ComponentMask(None=NamedRerouteUsage()) {g=true}
  n62 = SmoothStep(
   Min=MaterialFunctionCall(Float2=NamedRerouteUsage()) {materialFunction=&BreakOutFloat2Components}.G
   Value=ComponentMask(None=NamedRerouteUsage()) {g=true}
  )
  n63 = Add(A=Ceil(None=SmoothStep(Min=@n60.G, Value=@n61)), B=OneMinus(None=Ceil(None=@n62)))
  n64 = NamedRerouteUsage()
  n65 = NamedRerouteUsage()
  n66 = NamedRerouteUsage()
  n67 = NamedRerouteUsage()
  n68 = MaterialFunctionCall(
   A=Multiply(
    A=MaterialFunctionCall(Float2=@n64) {materialFunction=&BreakOutFloat2Components}.R
    B=MaterialFunctionCall(Float2=@n65) {materialFunction=&BreakOutFloat2Components}.G
   )
   B=Multiply(
    A=MaterialFunctionCall(Float2=@n66) {materialFunction=&BreakOutFloat2Components}.G
    B=MaterialFunctionCall(Float2=@n67) {materialFunction=&BreakOutFloat2Components}.R
   )
  ) {materialFunction=&MF_UI_IsGreater}
  NamedRerouteDeclaration(None=MaterialFunctionCall(ResultNumber=Add(A=@n63, B=@n68.Result)) {materialFunction=&MF_UI_WindngNumberConditions}.Result) {name=s1, nodeColor=<rgba 0 0 0 1>}
  n69 = MaterialFunctionCall(Float2=NamedRerouteUsage()) {materialFunction=&BreakOutFloat2Components}
  n70 = ComponentMask(None=NamedRerouteUsage()) {g=true}
  n71 = SmoothStep(
   Min=MaterialFunctionCall(Float2=NamedRerouteUsage()) {materialFunction=&BreakOutFloat2Components}.G
   Value=ComponentMask(None=NamedRerouteUsage()) {g=true}
  )
  n72 = Add(A=Ceil(None=SmoothStep(Min=@n69.G, Value=@n70)), B=OneMinus(None=Ceil(None=@n71)))
  n73 = NamedRerouteUsage()
  n74 = NamedRerouteUsage()
  n75 = NamedRerouteUsage()
  n76 = NamedRerouteUsage()
  n77 = MaterialFunctionCall(
   A=Multiply(
    A=MaterialFunctionCall(Float2=@n73) {materialFunction=&BreakOutFloat2Components}.R
    B=MaterialFunctionCall(Float2=@n74) {materialFunction=&BreakOutFloat2Components}.G
   )
   B=Multiply(
    A=MaterialFunctionCall(Float2=@n75) {materialFunction=&BreakOutFloat2Components}.G
    B=MaterialFunctionCall(Float2=@n76) {materialFunction=&BreakOutFloat2Components}.R
   )
  ) {materialFunction=&MF_UI_IsGreater}
  NamedRerouteDeclaration(None=MaterialFunctionCall(ResultNumber=Add(A=@n72, B=@n77.Result)) {materialFunction=&MF_UI_WindngNumberConditions}.Result) {name=s2, nodeColor=<rgba 0 0 0 1>}
 ]}
