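# Material function that draws one animated hit-marker tick shape for UI, outputting color and opacity.
# UVs are offset by position inputs plus a looping curve-atlas animation scaled by MoveTicks and direction, then scaled and rotated into a four-vertex polygon SDF.
# The SDF is stepped and remapped into fill and outline bands, each tinted by a two-color blend along a rotatable linear gradient.
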MaterialFunction /HitTicks/Materials/MF_UI_HitMarkerShape
graph {nodes [
  n1 = MaterialFunctionCall(Float4=FunctionInput() {bUsePreviewValueAsDefault=true, inputName=Vertex0, inputType=FunctionInput_Vector4}) {materialFunction=&BreakOutFloat4Components}
  n2 = MaterialFunctionCall(Float4=FunctionInput() {bUsePreviewValueAsDefault=true, inputName=Vertex1, inputType=FunctionInput_Vector4, sortPriority=1}) {materialFunction=&BreakOutFloat4Components}
  n3 = NamedRerouteUsage()
  n4 = VectorParameter() {parameterName=Scale, defaultValue=<rgba 0.15 0.15 0 1>}
  n5 = FunctionInput() {bUsePreviewValueAsDefault=true, description=Default 0, inputName=RotationAngle, inputType=FunctionInput_Scalar, sortPriority=5}
  n6 = MaterialFunctionCall!(
   UVs=MaterialFunctionCall(
    UVs=MaterialFunctionCall(UVs=@n3, Texture Scale=@n4.RGB) {materialFunction=&ScaleUVsByCenter}.UVs
    "Rotation Angle (0-1)"=@n5
   ) {materialFunction=&CustomRotator}.Rotated Values
   vertex_0=AppendVector(A=@n1.R, B=@n1.R)
   vertex_1=AppendVector(A=@n1.B, B=@n1.B)
   vertex_2=AppendVector(A=@n2.R, B=@n2.R)
   vertex_3=AppendVector(A=@n2.B, B=@n2.B)
   Outline Thickness
   Glow Min
   Glow Max
  ) {materialFunction=&/MF_UI_4VertexPolygon}
  n7 = OneMinus(None=@n6.SDF)
  n8 = Constant() {r=0.704}
  n9 = Constant() {r=0.9}
  n10 = Step(Y=@n6.SDF) {constX=0.1}
  n11 = Subtract(A=Divide(A=Subtract(A=@n7, B=@n8), B=Subtract(A=@n9, B=@n8)), B=@n10)
  n12 = Power(Base=LinearInterpolate(
    A=VectorParameter() {parameterName=Outline1, defaultValue=<rgba 1 1 1 1>}.RGB
    B=VectorParameter() {parameterName=Outline2, defaultValue=<rgba 1 1 1 1>}.RGB
    Alpha=NamedRerouteUsage()
   ))
  n13 = LinearInterpolate(
   A=VectorParameter() {parameterName=BGColor1, defaultValue=<rgba 1 1 1 1>}.RGB
   B=VectorParameter() {parameterName=BGColor2, defaultValue=<rgba 1 1 1 1>}.RGB
   Alpha=NamedRerouteUsage()
  )
  FunctionOutput(None=LinearInterpolate(A=Multiply(A=@n11, B=@n12), B=Power(Base=@n13), Alpha=@n10))
  FunctionOutput(None=Add(A=Saturate(None=@n11), B=@n10)) {outputName=Opacity}
  n14 = TextureCoordinate()
  n15 = ComponentMask(None=@n14) {r=true}
  n16 = CurveAtlasRowParameter(CurveTime=Frac(None=Time())) {atlas=&/CA_HitTicksAnimation, curve=&/C_HitTicks_Anim, parameterName=ShieldCrackAnimation}
  n17 = ScalarParameter() {parameterName=MoveTicks}
  n18 = FunctionInput() {
   bUsePreviewValueAsDefault=true
   description=Default 1
   inputName=MovementDirectionX
   inputType=FunctionInput_Scalar
   sortPriority=8
   previewValue=<1 0 0 1>
  }
  n19 = Add(
   A=FunctionInput() {bUsePreviewValueAsDefault=true, inputName=ShapeXPosition, inputType=FunctionInput_Scalar, sortPriority=6}
   B=Multiply(A=Multiply(A=@n16.B, B=@n17), B=@n18)
  )
  n20 = ComponentMask(None=@n14) {g=true}
  n21 = FunctionInput() {bUsePreviewValueAsDefault=true, description=Default 1, inputName=ShapeYPosition, inputType=FunctionInput_Scalar, sortPriority=7}
  n22 = FunctionInput() {
   bUsePreviewValueAsDefault=true
   description=1 for moving left and -1 for moving Right
   inputName=MovementDirectionY
   inputType=FunctionInput_Scalar
   sortPriority=9
   previewValue=<1 0 0 1>
  }
  n23 = Add(A=@n21, B=Multiply(A=Multiply(A=@n16.B, B=@n17), B=@n22))
  NamedRerouteDeclaration(None=AppendVector(A=Add(A=@n15, B=@n19) {constB=0.2}, B=Add(A=@n20, B=@n23) {constB=0.2})) {name=CustomUV, nodeColor=<rgba 0 0 0 1>}
  n24 = MaterialFunctionCall(UVs=NamedRerouteUsage(), Texture Scale=Constant2Vector() {g=0.1, r=0.1}) {materialFunction=&ScaleUVsByCenter}
  n25 = FunctionInput() {bUsePreviewValueAsDefault=true, inputName=GradientRotation, inputType=FunctionInput_Scalar, sortPriority=30}
  NamedRerouteDeclaration(None=MaterialFunctionCall(UV Channel=MaterialFunctionCall(UVs=@n24.UVs, "Rotation Angle (0-1)"=@n25) {materialFunction=&CustomRotator}.Rotated Values) {materialFunction=&LinearGradient}.UGradient) {name=LerpAlpha, nodeColor=<rgba 0 0 0 1>}
 ]}
