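# Translucent UI material that draws a four-corner hit marker whose shape morphs between Shield, Health and Elim states.
# The Shape parameter blends three presets of packed vertex coordinates through nested lerps, feeding two named reroutes.
# Four MF_UI_HitMarkerShape calls place the corners, each with its own position, rotation, movement direction and gradient rotation.
# Their results combine by Max, while their opacities are summed and scaled by an Opacity parameter.
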Material /HitTicks/Materials/M_UI_HitMarker
graph {
 properties {blendMode=BLEND_Translucent, materialDomain=MD_UI, pixelDepthOffsetMode=PDOM_Legacy, usedShadingModels=Default Lit}
 nodes [
  n1 = MaterialFunctionCall!(
   Vertex0=NamedRerouteUsage()
   Vertex1=NamedRerouteUsage()
   RotationAngle
   ShapeXPosition=ScalarParameter() {defaultValue=0.18, parameterName=Corner1X}
   ShapeYPosition=ScalarParameter() {defaultValue=0.18, parameterName=Corner1Y}
   MovementDirectionX
   MovementDirectionY
   GradientRotation=ScalarParameter() {parameterName=Gradient1Rot}
  ) {materialFunction=&/MF_UI_HitMarkerShape}
  n2 = MaterialFunctionCall(
   Vertex0=NamedRerouteUsage()
   Vertex1=NamedRerouteUsage()
   RotationAngle=ScalarParameter() {defaultValue=-0.25, parameterName=PointBRot}
   ShapeXPosition=ScalarParameter() {defaultValue=-0.18, parameterName=Corner2X}
   ShapeYPosition=ScalarParameter() {defaultValue=0.18, parameterName=Corner2Y}
   MovementDirectionX=Constant() {r=-1}
   GradientRotation=ScalarParameter() {parameterName=Gradient2Rot}
  ) {materialFunction=&MF_UI_HitMarkerShape}
  n3 = MaterialFunctionCall(
   Vertex0=NamedRerouteUsage()
   Vertex1=NamedRerouteUsage()
   RotationAngle=ScalarParameter() {defaultValue=0.25, parameterName=PointCRot}
   ShapeXPosition=ScalarParameter() {defaultValue=0.18, parameterName=Corner3X}
   ShapeYPosition=ScalarParameter() {defaultValue=-0.18, parameterName=Corner3Y}
   MovementDirectionY=Constant() {r=-1}
   GradientRotation=ScalarParameter() {parameterName=Gradient3Rot}
  ) {materialFunction=&MF_UI_HitMarkerShape}
  n4 = Constant() {r=-1}
  n5 = MaterialFunctionCall(
   Vertex0=NamedRerouteUsage()
   Vertex1=NamedRerouteUsage()
   RotationAngle=ScalarParameter() {defaultValue=0.5, parameterName=PointDRot}
   ShapeXPosition=ScalarParameter() {defaultValue=-0.18, parameterName=Corner4X}
   ShapeYPosition=ScalarParameter() {defaultValue=-0.18, parameterName=Corner4Y}
   MovementDirectionX=@n4
   MovementDirectionY=@n4
   GradientRotation=ScalarParameter() {parameterName=Gradient4Rot}
  ) {materialFunction=&MF_UI_HitMarkerShape}
  Max(A=Max(A=@n1.Result, B=@n2.Result), B=Max(A=@n3.Result, B=@n5.Result))
  Multiply(A=Add(A=Add(A=@n1.Opacity, B=@n2.Opacity), B=Add(A=@n3.Opacity, B=@n5.Opacity)), B=ScalarParameter() {defaultValue=1, parameterName=Opacity})
  n6 = Constant4Vector() {constant=<rgba 0.1 0.1 0.52 -0.65>}
  n7 = Constant4Vector() {constant=<rgba -0.45 -0.45 -0.2 -0.2>}
  n8 = ScalarParameter() {defaultValue=1, desc="0-Shield, 0.5-Health, 1-Elim", parameterName=Shape}
  n9 = Clamp(None=Multiply(A=@n8) {constB=2})
  n10 = Constant4Vector() {constant=<rgba -1 -1 0.175 -0.175>}
  n11 = Clamp(None=Multiply(A=Add(A=@n8) {constB=-0.5}) {constB=2})
  NamedRerouteDeclaration(None=LinearInterpolate(A=LinearInterpolate(A=@n6.RGBA, B=@n7.RGBA, Alpha=@n9), B=LinearInterpolate(A=@n7.RGBA, B=@n10.RGBA, Alpha=@n11), Alpha=@n8) {desc="Blend it!"}) {name=FirstFourPints, nodeColor=<rgba 0 0 0 1>}
  n12 = Constant4Vector() {constant=<rgba -0.2 -0.2 -0.7 0.7>}
  n13 = Constant4Vector() {constant=<rgba 0.35 0.4 0.4 0.35>}
  n14 = ScalarParameter() {defaultValue=1, desc="0-Shield, 0.5-Health, 1-Elim", parameterName=Shape}
  n15 = Saturate(None=Multiply(A=@n14) {constB=2})
  n16 = Constant4Vector() {constant=<rgba 0.7 0.65 -0.15 0.15>}
  n17 = Clamp(None=Multiply(A=Add(A=@n14) {constB=-0.5}) {constB=2})
  NamedRerouteDeclaration(None=LinearInterpolate(A=LinearInterpolate(A=@n12.RGBA, B=@n13.RGBA, Alpha=@n15), B=LinearInterpolate(A=@n13.RGBA, B=@n16.RGBA, Alpha=@n17), Alpha=@n14) {desc="Blend it!"}) {name=SecondFourPoints, nodeColor=<rgba 0 0 0 1>}
 ]
}
