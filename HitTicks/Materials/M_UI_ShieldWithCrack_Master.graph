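# Translucent UI master material for a cracked shield icon, with an animation that splits it into two pieces.
# Frac of Time, scaled by the EnableCracks parameter, samples the ShieldAnim curve atlas row to push two UV sets in opposite diagonal directions.
# Each set is scaled about its center and samples one channel (R or G) of the cracked icon.
# The same moving UVs also sample an ice gradient, tinted cyan as a faint additive layer.
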Material /HitTicks/Materials/M_UI_ShieldWithCrack_Master
graph {
 properties {blendMode=BLEND_Translucent, materialDomain=MD_UI, pixelDepthOffsetMode=PDOM_Legacy, usedShadingModels=Default Lit}
 nodes [
  n1 = TextureCoordinate()
  n2 = ComponentMask(None=@n1) {r=true}
  n3 = Time()
  n4 = CurveAtlasRowParameter(CurveTime=Multiply(A=Frac(None=@n3), B=ScalarParameter() {defaultValue=0.5, parameterName=EnableCracks})) {atlas=&/CA_HitTicksAnimation, curve=&/C_ShieldAnim, defaultValue=1, parameterName=ShieldAnim}
  n5 = ComponentMask(None=@n1) {g=true}
  n6 = Multiply(A=@n4.R) {constB=-1}
  n7 = Constant() {r=0.8}
  n8 = MaterialFunctionCall(UVs=AppendVector(A=Add(A=@n2, B=@n4.R) {constB=0}, B=Add(A=@n5, B=@n6) {constB=0}), Texture Scale=@n7) {materialFunction=&ScaleUVsByCenter}
  n9 = TextureSample(UVs=@n8.UVs) {texture=&/IceGradient3}
  Add(
   A=ComponentMask(None=@n1) {g=true}
   B=Multiply(A=Constant3Vector() {constant=<rgba 0.0319 0.8469 0.9911 1>}, B=Multiply(A=@n9.B) {constA=0.5, constB=0.2}) {constB=0.1}
  )
  n10 = ComponentMask(None=@n1) {r=true}
  n11 = Multiply(A=@n4.R) {constB=-1.2}
  n12 = ComponentMask(None=@n1) {g=true}
  n13 = MaterialFunctionCall(UVs=AppendVector(A=Add(A=@n10, B=@n11), B=Add(A=@n12, B=@n4.R) {constB=0}), Texture Scale=@n7) {materialFunction=&ScaleUVsByCenter}
  n14 = MaterialFunctionCall(UVs=@n13.UVs, Texture Scale=Constant() {r=1.6}) {materialFunction=&ScaleUVsByCenter}
  n15 = Multiply(
   A=Constant3Vector() {constant=<rgba 0.0319 0.8469 0.9911 1>}
   B=Multiply(A=TextureSample(UVs=@n14.UVs) {texture=&IceGradient3}.B) {constA=0.5}
  ) {constB=0.1}
  Add(B=@n15)
  Add(
   A=TextureSample(UVs=@n13.UVs) {samplerSource=SSM_Clamp_WorldGroupSettings, texture=&/ShieldCrackedIcon}.R
   B=TextureSample(UVs=@n8.UVs) {samplerSource=SSM_Clamp_WorldGroupSettings, texture=&ShieldCrackedIcon}.G
  )
  Add(A=@n15)
 ]
}
