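# WBP_HitTicksWidget is a hit-tick HUD widget that shows hit feedback as a centered text block, apparently as a placeholder.
# HandleHitTickEvent builds a string array with damage and optional shield, critical and fatal labels, then writes it line by line to TempText.
# On Construct it broadcasts a field change for UseNativeMarkerTag, a gameplay tag defaulting to the shotgun item.
# The shield break handler is empty.
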WidgetBlueprint /HitTicks/WBP_HitTicksWidget
graphs [
 EventGraph {nodes [
   Event Construct()
   > Broadcast Field Value Changed!(execute, self:Widget Object Reference, FieldId:Field Notification Id Structure=(FieldName="UseNativeMarkerTag")) -> (then)
   Event PreConstruct()
   Event Tick()
  ]}
 HandleHitTickEvent {nodes [
   n1 = HandleHitTickEvent!() -> (
    then
    Event_Damage:double=~
    Event_AccumulatedDamage:double=~
    Event_bShield:bool=~
    Event_bCritical:bool=~
    Event_bFatal:bool=~
    Event_Target:Actor Object Reference
   ) {is=title K2Node_FunctionEntry}
   > n2 = Sequence!(execute) -> (then_0, then_1, then_2:Exec, then_3:Exec, then_4:Exec)
   > Branch()
   > Add(
    TargetArray=Get HitTickDescription() -> (HitTickDescription:Array of Strings)
    NewItem=Append(A=MakeLiteralString(Value="Damage: "), B="To String (Float)"(InDouble=@n1))
   ) {is=title K2Node_CallArrayFunction}
   Branch(execute=@n2.then_1, Condition=@n1.Event_bShield)
   > Add(TargetArray=Get HitTickDescription(), NewItem=MakeLiteralString(Value=Shield Hit))
   Branch(execute=@n2.then_2, Condition=@n1.Event_bCritical)
   > Add(TargetArray=Get HitTickDescription(), NewItem=MakeLiteralString(Value=Critical Hit))
   Branch(execute=@n2.then_3, Condition=@n1.Event_bFatal)
   > Add(TargetArray=Get HitTickDescription(), NewItem=MakeLiteralString(Value=Fatal Hit))
   n3 = Get HitTickDescription()
   Set Text(
    execute=@n2.then_4
    Text="To Text (String)"(InString=JoinStringArray(SourceArray=@n3, Separator="\r\n"))
    self=Get TempText() -> (TempText:Common Text Object Reference)
   )
  ]}
 HandleShieldBreakEvent {nodes [HandleShieldBreakEvent!() -> (then, bIsOvershield:bool=~) {is=title K2Node_FunctionEntry}]}
]
variables [UseNativeMarkerTag {type=GameplayTag, default {tagName=Item.Weapon.Ranged.Shotgun}}]
widgets {
 info {inheritedWidgetCount=0, parentClass=&UserWidget, rootWidgetClass=&Overlay, widgetCount=2}
 widgets [
  Overlay_0 {widgetClassPath=&Overlay, widget=Overlay_0, properties {bOverride_Cursor=true, visibility=Visible}}
  TempText {
   widgetClassPath=&CommonTextBlock
   bIsVariable=true
   parent=Overlay_0
   widget=TempText
   properties {
    style=&TextStyle-Base_C
    text=Text Block
    font {
     bForceMonospaced=false
     bMaterialIsStencil=false
     fontMaterial=null
     fontObject=&NotoSans
     letterSpacing=0
     monospacedWidth=1
     size=24
     skewAmount=0
     typefaceFontName=NotoSans-Regular
     outlineSettings {
      bApplyOutlineToDropShadows=false
      bMiteredCorners=false
      bSeparateFillAlpha=false
      outlineMaterial=null
      outlineSize=0
      outlineColor=<rgba 0 0 0 1>
     }
    }
    shadowOffset=<0 0>
   }
   slot {class=OverlaySlot, properties {horizontalAlignment=HAlign_Center, verticalAlignment=VAlign_Center}}
  }
 ]
}
